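# Fortnite UI watermark widget, based on FortBuildWatermark, with a row of translucent text for build, MCP version, MCP time offset and build ID override.
# On Construct it calls UpdateWatermark, then binds to the party context's player-state-changed and party-joined events.
# Both handlers simply call UpdateWatermark again; the graph never sets the text itself.
# On Destruct it checks the party context is valid and removes both bindings.
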WidgetBlueprint /Game/UI/Foundation/Watermark/BuildWatermark
graphs [EventGraph {nodes [
   n1 = HandlePartyJoined() {is=title K2Node_CustomEvent}
   UpdateWatermark(execute=[@n1.then, @n2.then], self:Fort Build Watermark Object Reference)
   n2 = HandlePlayerStateChanged!() -> (OutputDelegate, then, PlayerInfo:"Fort Team Member Info Structure (by ref)") {is=title K2Node_CustomEvent}
   Event Construct()
   > UpdateWatermark()
   n3 = Create Event()
   n4 = Get FortPartyContext!() -> (ReturnValue:Fort Party Context Object Reference) {is=title K2Node_GetBlueprintContextEx}
   > Assign On Player State Changed!(execute:Exec, self:Fort Party Context Object Reference=@n4, Delegate:"Delegate (by ref)"=@n3) -> (then:Exec) {is=title K2Node_AssignDelegate}
   n5 = Create Event()
   > Bind Event to On Party Joined(self:Fort Party Context Object Reference=@n4, Delegate=@n5) {is=title K2Node_AddDelegate}
   Event Destruct()
   > Is Valid(InputObject=@n4)
   > Unbind Event from On Player State Changed(self:Fort Party Context Object Reference=@n4, Delegate=@n3) {is=title K2Node_RemoveDelegate}
   > Unbind Event from On Party Joined(self:Fort Party Context Object Reference=@n4, Delegate=@n5) {is=title K2Node_RemoveDelegate}
  ]}]
variables []
widgets {
 info {inheritedWidgetCount=5, parentClass=/Script/FortniteUI.FortBuildWatermark, rootWidgetClass=&HorizontalBox, widgetCount=6}
 widgets [
  HorizontalBox_0 {widgetClassPath=&HorizontalBox, widget=HorizontalBox_0}
  Text_BuildWatermark {
   widgetClassPath=&TextBlock
   bInherited=true
   bIsVariable=true
   parent=HorizontalBox_0
   widget=Text_BuildWatermark
   properties {
    justification=Center
    text=BUILD WATERMARK
    visibility=SelfHitTestInvisible
    colorAndOpacity {colorUseRule=UseColor_Specified, specifiedColor=<rgba 1 1 1 0.5>}
    font {
     bForceMonospaced=false
     bMaterialIsStencil=false
     fontMaterial=null
     fontObject=&NotoSans
     letterSpacing=0
     monospacedWidth=1
     size=10
     skewAmount=0
     typefaceFontName=NotoSans-Bold
     outlineSettings {
      bApplyOutlineToDropShadows=false
      bMiteredCorners=false
      bSeparateFillAlpha=false
      outlineMaterial=null
      outlineSize=0
      outlineColor=<rgba 0 0 0 1>
     }
    }
   }
   slot {class=HorizontalBoxSlot}
  }
  Text_McpVersion {
   widgetClassPath=&TextBlock
   bInherited=true
   bIsVariable=true
   parent=HorizontalBox_0
   widget=Text_McpVersion
   properties {
    justification=Center
    text=MCP VERSION
    visibility=SelfHitTestInvisible
    colorAndOpacity {colorUseRule=UseColor_Specified, specifiedColor=<rgba 1 1 1 0.5>}
    font {
     bForceMonospaced=false
     bMaterialIsStencil=false
     fontMaterial=null
     fontObject=&NotoSans
     letterSpacing=0
     monospacedWidth=1
     size=10
     skewAmount=0
     typefaceFontName=NotoSans-Bold
     outlineSettings {
      bApplyOutlineToDropShadows=false
      bMiteredCorners=false
      bSeparateFillAlpha=false
      outlineMaterial=null
      outlineSize=0
      outlineColor=<rgba 0 0 0 1>
     }
    }
   }
   slot {class=HorizontalBoxSlot, properties {padding=<ltrb 12 0 0 0>}}
  }
  Text_CalendarOffset {
   widgetClassPath=&TextBlock
   bInherited=true
   bIsVariable=true
   parent=HorizontalBox_0
   widget=Text_CalendarOffset
   properties {
    justification=Center
    text=MCP Time Offset:
    visibility=SelfHitTestInvisible
    colorAndOpacity {colorUseRule=UseColor_Specified, specifiedColor=<rgba 1 1 1 0.5>}
    font {
     bForceMonospaced=false
     bMaterialIsStencil=false
     fontMaterial=null
     fontObject=&NotoSans
     letterSpacing=0
     monospacedWidth=1
     size=10
     skewAmount=0
     typefaceFontName=NotoSans-Bold
     outlineSettings {
      bApplyOutlineToDropShadows=false
      bMiteredCorners=false
      bSeparateFillAlpha=false
      outlineMaterial=null
      outlineSize=0
      outlineColor=<rgba 0 0 0 1>
     }
    }
   }
   slot {class=HorizontalBoxSlot, properties {padding=<ltrb 12 0 0 0>}}
  }
  Text_BuildIDOverride {
   widgetClassPath=&TextBlock
   bInherited=true
   bIsVariable=true
   parent=HorizontalBox_0
   widget=Text_BuildIDOverride
   properties {
    justification=Center
    text=Build ID Override:
    visibility=SelfHitTestInvisible
    colorAndOpacity {colorUseRule=UseColor_Specified, specifiedColor=<rgba 1 1 1 0.5>}
    font {
     bForceMonospaced=false
     bMaterialIsStencil=false
     fontMaterial=null
     fontObject=&NotoSans
     letterSpacing=0
     monospacedWidth=1
     size=10
     skewAmount=0
     typefaceFontName=NotoSans-Bold
     outlineSettings {
      bApplyOutlineToDropShadows=false
      bMiteredCorners=false
      bSeparateFillAlpha=false
      outlineMaterial=null
      outlineSize=0
      outlineColor=<rgba 0 0 0 1>
     }
    }
   }
   slot {class=HorizontalBoxSlot, properties {padding=<ltrb 12 0 0 0>}}
  }
  TutorialHighlight {widgetClassPath=&TutorialHighlight}
 ]
}
